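# Conversation database for the BistroAstronaut NPC, who introduces herself as Ariana and greets players before offering shared NPC options.
# The entry point leads to one Speech node with a default line and contextual greetings gated on play history, first meeting, Bars, Sideways visit or outfit.
# Curve-table rows set the priority of each contextual greeting.
# The speech then branches through link nodes into the shared NPCLibrary Quests and BasicServices conversations.
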ConversationDatabase /NPCLibrary/NPCs/BistroAstronaut/BistroAstronaut/NPCLibrary_Conversation_BistroAstronaut
entry_points [{tag=Conversation.Entry.BistroAstronaut, destinations [#1]}]
nodes [
 ConversationEntryPointNode_7 {
  id=#1
  class=ConversationEntryPointNode
  outputs [#2]
  properties {entryTag {tagName=Conversation.Entry.BistroAstronaut}}
 }
 Speech_C_4 {
  id=#2
  class=Speech_C
  outputs [#3, #4]
  properties {
   nodeName=""
   generalConfig {
    defaultMessage=You stayin' safe out there?
    contextualMessages [
     {
      message="You're back! I could use some help with these monsters."
      contextRequirements [{requirement {class=FortControllerRequirement_CheckTimeSinceLastPlayedMatch}}]
      priority {curve {rowName=Tandem.Speech.Priority.WhileSincePlayed}}
     }
     {
      message="You made it out of the Sideways! How'd it feel?"
      contextRequirements [{requirement {class=FortControllerRequirement_HasOwnedTag, properties {ownedTag {tagName=Granted.Athena.Sideways.Encounter.Visited}}}}]
      priority {curve {rowName=Tandem.Speech.Priority.VisitedSpecificPOI}}
     }
     {
      message="Hi! We haven't met. I'm Ariana!"
      contextRequirements [{requirement {class=FortControllerRequirement_IsFirstEverConversationWithNPC}}]
      priority {curve {rowName=Tandem.Speech.Priority.FirstEverInteraction}}
     }
     {
      message=Is that a Bar count or a phone number?
      contextRequirements [{requirement {class=FortControllerRequirement_HasBars}}]
      priority {curve {rowName=Tandem.Speech.Priority.HighTotalBars}}
     }
     {
      message="Hey! Lookin' good!"
      contextRequirements [{requirement {
         class=FortControllerRequirement_HasCID
         properties {allowedCIds [
           /BRCosmetics/Athena/Items/Cosmetics/Characters/CID_A_227_Athena_Commando_F_BistroAstronaut_JJLK5
           /BRCosmetics/Athena/Items/Cosmetics/Characters/CID_A_158_Athena_Commando_F_Buffet_YC20H
          ]}
        }}]
      priority {curve {rowName=Tandem.Speech.Priority.OutfitReaction}}
     }
    ]
   }
  }
 }
 ConversationLinkNode_0 {id=#4, class=ConversationLinkNode, properties {remoteEntryTag {tagName=Conversation.Entry.BasicServices}}}
 ConversationLinkNode_1 {id=#3, class=ConversationLinkNode, properties {remoteEntryTag {tagName=Conversation.Entry.NPCLibrary.Quests}}}
]
